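# Stereo MetaSound source for a music encounter's diegetic loop, playing a supplied wave asset with a fade-in and no clock input.
# OnPlay starts the Wave Player and switches a float from 0 to 1, which InterpTo smooths over FadeInTime.
# That ramp multiplies the left and right channels into the outputs.
# The player's Loop default holds a stray builder-result string instead of a boolean, though the source is marked looping.
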MetaSoundSource /MusicEncounter/MetaSounds/MSS_MusicEncounter_DiageticLoop_NoClock
inputs [
 FadeInTime {type=Time}
 Music.WaveAsset {type=WaveAsset}
 UE.Source.OnPlay {type=Trigger}
]
interfaces [UE.OutputFormat.Stereo:1.0, UE.Source:1.1]
outputs [
 "UE.OutputFormat.Stereo.Audio:0" {type=Audio}
 "UE.OutputFormat.Stereo.Audio:1" {type=Audio}
]
pages [Default {
  input_defaults {FadeInTime=0, Music.WaveAsset=null, UE.Source.OnPlay=null}
  nodes [
   n1 = UE.InterpTo:Audio(Interp Time=^FadeInTime, Target=Value.Value:Float(Set=^UE.Source.OnPlay) {also {defaults {Init Value=0, Target Value=1}}}.Output Value)
   WavePlayer_0 = UE.Wave Player:Stereo(Wave Asset=^Music.WaveAsset, Play=^UE.Source.OnPlay) {also {defaults {Loop="<MetaSoundBuilderResult.SUCCEEDED: 0>"}, version [1,1]}}
   Multiply_0 = UE.Multiply:Audio by Float(AdditionalOperands=@n1.Value, PrimaryOperand=@WavePlayer_0.Out Left)
   Multiply_1 = UE.Multiply:Audio by Float(AdditionalOperands=@n1.Value, PrimaryOperand=@WavePlayer_0.Out Right)
  ]
  links [{dst="@UE.OutputFormat.Stereo.Audio:0", src=Multiply_0.Out}, {dst="@UE.OutputFormat.Stereo.Audio:1", src=Multiply_1.Out}]
 }]
properties {
 assetImportData=&self:AssetImportData
 bLooping=true
 duration=10000
 numChannels=2
 outputFormat=Stereo
 sampleRate=48000
}
