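# Stereo MetaSound source for NPC encounter music that moves between Idle, Combat and Defeated phases.
# A tempo-driven Harmonix metronome quantizes phase changes, while beat counters delay loop crossfades and transition waves bridge Idle and Combat.
# The Defeated wave fades out the other layers, feeds loot, ambient and weapon sidechain buses, then finishes the source.
# Output gain follows a weapon-fire envelope modulator and a pause fade, which also finishes the source.
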MetaSoundSource /MusicEncounter/MetaSounds/MSS_MusicEncounter_NPC_Default
inputs [
 Combat.BeatOffset {type=Int32}
 Combat.ElseWave {type=WaveAsset}
 Combat.FadeInTime {type=Time}
 Combat.FadeOutTime {type=Time}
 Combat.PercWave {type=WaveAsset}
 Combat.TransitionVolume {type=Float}
 Combat.TransitionWave {type=WaveAsset}
 Combat.Volume {type=Float}
 Defeated.BeatOffset {type=Int32}
 Defeated.Volume {type=Float}
 Defeated.Wave {type=WaveAsset}
 Idle.BeatOffset {type=Int32}
 Idle.FadeInTime {type=Time}
 Idle.FadeOutTime {type=Time}
 Idle.TransitionVolume {type=Float}
 Idle.TransitionWave {type=WaveAsset}
 Idle.Volume {type=Float}
 Idle.Wave {type=WaveAsset}
 MusicEncounter.bIsPaused {type=Bool}
 MusicEncounter.OutputVolume {type=Float}
 MusicEncounter.PauseFadeOutTime {type=Time}
 MusicEncounter.Phase {type=Int32}
 MusicEncounter.Tempo {type=Float}
 UE.Source.OnPlay {type=Trigger}
]
interfaces [UE.OutputFormat.Stereo:1.0, UE.Source.OneShot:1.0, UE.Source:1.1]
outputs [
 "UE.OutputFormat.Stereo.Audio:0" {type=Audio}
 "UE.OutputFormat.Stereo.Audio:1" {type=Audio}
 UE.Source.OneShot.OnFinished {type=Trigger}
]
pages [Default {
  input_defaults {
   Combat.BeatOffset=1
   Combat.ElseWave=null
   Combat.FadeInTime=1
   Combat.FadeOutTime=1
   Combat.PercWave=null
   Combat.TransitionVolume=0
   Combat.TransitionWave=null
   Combat.Volume=0
   Defeated.BeatOffset=1
   Defeated.Volume=0
   Defeated.Wave=null
   Idle.BeatOffset=1
   Idle.FadeInTime=1
   Idle.FadeOutTime=1
   Idle.TransitionVolume=0
   Idle.TransitionWave=null
   Idle.Volume=0
   Idle.Wave=null
   MusicEncounter.OutputVolume=0
   MusicEncounter.PauseFadeOutTime=3
   MusicEncounter.Phase=0
   MusicEncounter.Tempo=120
   MusicEncounter.bIsPaused=null
   UE.Source.OnPlay=null
  }
  nodes [
   n1 = Max.Max:Int32(A=^Defeated.BeatOffset) {also {defaults {B=1}}}
   n2 = HarmonixNodes.Metronome(Tempo=^MusicEncounter.Tempo, Transport=HarmonixNodes.TriggerToTransport(Play=^UE.Source.OnPlay).Transport) {also {defaults {Time Sig. Numerator=1}}}
   n3 = VariableAccessor.MIDIClock(Variable=VariableMutator.MIDIClock(Value=@n2.MIDI Clock, Variable=InitVariable.MIDIClock().Variable).Variable)
   n4 = VariableAccessor.MIDIClock(Variable=@n3.Variable)
   n5 = VariableAccessor.MIDIClock(Variable=@n4.Variable)
   n6 = InitVariable.Trigger()
   TriggerOnValueChange_1 = UE.Trigger On Value Change:Int32(Value=^MusicEncounter.Phase)
   TriggerSelect3_0 = "UE.Trigger Select (3)"(Index=UE.Reroute(Value=^MusicEncounter.Phase).Value, In=@TriggerOnValueChange_1.Trigger)
   n7 = HarmonixNodes.MidiQuantizeTriggerNode(
    MIDI Clock=@n5.Value
    Trigger Input=VariableAccessor.Trigger(Variable=VariableMutator.Trigger(Variable=@n6.Variable, Value=@TriggerSelect3_0.Out 2).Variable).Value
   ) {also {version [0,1]}}
   n8 = InitVariable.Trigger()
   n9 = VariableAccessor.Trigger(Variable=VariableMutator.Trigger(Value=HarmonixNodes.MidiClockSubdivisionTrigger(MIDI Clock=@n2.MIDI Clock).Trigger Out, Variable=InitVariable.Trigger().Variable).Variable)
   n10 = VariableAccessor.Trigger(Variable=@n9.Variable)
   n11 = VariableAccessor.Trigger(Variable=@n10.Variable)
   TriggerControl_2 = UE.Trigger Control(
    Open=VariableAccessor.Trigger(Variable=VariableMutator.Trigger(Value=@n7.Trigger Out, Variable=@n8.Variable).Variable).Value
    Trigger In=@n11.Value
   )
   TriggerCounter_0 = UE.Trigger Counter(Reset=@n7.Trigger Out, In=@TriggerControl_2.Trigger Out)
   n12 = VariableAccessor.Trigger(Variable=VariableMutator.Trigger(
     Variable=InitVariable.Trigger().Variable
     Value=TriggerCompare.Clamp:Int32(B=@n1.Value, A=@TriggerCounter_0.Count, Compare=@TriggerCounter_0.On Trigger).True
    ).Variable)
   n13 = VariableAccessor.Trigger(Variable=@n12.Variable)
   WavePlayer_5 = UE.Wave Player:Stereo(Wave Asset=^Defeated.Wave, Play=VariableAccessor.Trigger(Variable=@n13.Variable).Value) {also {version [1,1]}}
   AudioBusWriter2_0 = "UE.Audio Bus Writer (2)"(In 0=@WavePlayer_5.Out Left, In 1=@WavePlayer_5.Out Right) {also {defaults {Audio Bus=&AB_Sidechain_FortLoot_Vol}}}
   AudioBusWriter2_1 = "UE.Audio Bus Writer (2)"(In 0=@WavePlayer_5.Out Left, In 1=@WavePlayer_5.Out Right) {also {defaults {Audio Bus=&AB_Sidechain_FortAmbient_Vol}}}
   AudioBusWriter2_2 = "UE.Audio Bus Writer (2)"(In 0=@WavePlayer_5.Out Left, In 1=@WavePlayer_5.Out Right) {also {defaults {Audio Bus=&AB_Sidechain_FortWeapon_Vol}}}
   DecibelstoLinearGain_0 = UE.Decibels to Linear Gain:Float(Decibels=^Idle.TransitionVolume)
   DecibelstoLinearGain_1 = UE.Decibels to Linear Gain:Float(Decibels=^Combat.TransitionVolume)
   DecibelstoLinearGain_4 = UE.Decibels to Linear Gain:Float(Decibels=^Combat.Volume)
   n14 = InitVariable.Trigger()
   TriggerOnce_0 = UE.Trigger Once(Trigger In=VariableAccessor.Trigger(Variable=@n11.Variable).Value)
   TriggerSelect3_1 = "UE.Trigger Select (3)"(
    In=HarmonixNodes.MidiQuantizeTriggerNode(Trigger Input=@TriggerOnce_0.Trigger Out, MIDI Clock=VariableAccessor.MIDIClock(Variable=@n5.Variable).Value) {also {version [0,1]}}.Trigger Out
    Index=UE.Reroute(Value=^MusicEncounter.Phase).Value
   )
   TriggerAnyTriggerAny2_0 = "TriggerAny.Trigger Any (2)"(In 1=@TriggerSelect3_0.Out 0, In 0=@TriggerSelect3_1.Out 0)
   n15 = HarmonixNodes.MidiQuantizeTriggerNode(
    MIDI Clock=@n3.Value
    Trigger Input=VariableAccessor.Trigger(Variable=VariableMutator.Trigger(Variable=@n14.Variable, Value=@TriggerAnyTriggerAny2_0.Out).Variable).Value
   ) {also {version [0,1]}}
   n16 = TriggerCompare.Clamp:Int32(A=^MusicEncounter.Phase, Compare=@n15.Trigger Out)
   n17 = InitVariable.Trigger()
   TriggerAnyTriggerAny2_1 = "TriggerAny.Trigger Any (2)"(In 1=@TriggerSelect3_0.Out 1, In 0=@TriggerSelect3_1.Out 1)
   n18 = HarmonixNodes.MidiQuantizeTriggerNode(
    MIDI Clock=@n4.Value
    Trigger Input=VariableAccessor.Trigger(Variable=VariableMutator.Trigger(Variable=@n17.Variable, Value=@TriggerAnyTriggerAny2_1.Out).Variable).Value
   ) {also {version [0,1]}}
   n19 = VariableAccessor.Trigger(Variable=VariableMutator.Trigger(Value=@n18.Trigger Out, Variable=InitVariable.Trigger().Variable).Variable)
   n20 = VariableAccessor.Trigger(Variable=@n19.Variable)
   TriggerAnyTriggerAny2_4 = "TriggerAny.Trigger Any (2)"(In 1=@n12.Value, In 0=@n20.Value)
   n21 = VariableAccessor.Trigger(Variable=VariableMutator.Trigger(Variable=InitVariable.Trigger().Variable, Value=@n16.False).Variable)
   n22 = VariableAccessor.Trigger(Variable=@n21.Variable)
   TriggerControl_0 = UE.Trigger Control(Close=@TriggerAnyTriggerAny2_4.Out, Trigger In=@n9.Value, Open=@n22.Value)
   TriggerCounter_2 = UE.Trigger Counter(Reset=@n15.Trigger Out, In=@TriggerControl_0.Trigger Out)
   TriggerAnyTriggerAny2_3 = "TriggerAny.Trigger Any (2)"(
    In 1=@n16.True
    In 0=TriggerCompare.Clamp:Int32(B=Max.Max:Int32(A=^Idle.BeatOffset) {also {defaults {B=1}}}.Value, A=@TriggerCounter_2.Count, Compare=@TriggerCounter_2.On Trigger).True
   )
   n23 = VariableAccessor.Trigger(Variable=VariableMutator.Trigger(Variable=InitVariable.Trigger().Variable, Value=@TriggerAnyTriggerAny2_3.Out).Variable)
   n24 = VariableAccessor.Trigger(Variable=@n23.Variable)
   n25 = Max.Max:Int32(A=^Combat.BeatOffset) {also {defaults {B=1}}}
   TriggerAnyTriggerAny2_2 = "TriggerAny.Trigger Any (2)"(In 0=VariableAccessor.Trigger(Variable=@n22.Variable).Value, In 1=@n13.Value)
   TriggerControl_1 = UE.Trigger Control(Close=@TriggerAnyTriggerAny2_2.Out, Trigger In=@n10.Value, Open=VariableAccessor.Trigger(Variable=@n20.Variable).Value)
   TriggerCounter_1 = UE.Trigger Counter(Reset=@n18.Trigger Out, In=@TriggerControl_1.Trigger Out)
   n26 = VariableAccessor.Trigger(Variable=VariableMutator.Trigger(
     Variable=InitVariable.Trigger().Variable
     Value=TriggerCompare.Clamp:Int32(B=@n25.Value, A=@TriggerCounter_1.Count, Compare=@TriggerCounter_1.On Trigger).True
    ).Variable)
   n27 = Value.Value:Float(Reset=VariableAccessor.Trigger(Variable=@n24.Variable).Value, Set=@n26.Value) {also {defaults {Init Value=0, Target Value=1}}}
   TriggerRouteTriggerRouteTime2_0 = "TriggerRoute.Trigger Route (Time, 2):Time"(Value 0=^Combat.FadeInTime, Value 1=^Idle.FadeOutTime, Set 1=@n27.On Reset, Set 0=@n27.On Set)
   n28 = VariableAccessor.Float(Variable=VariableMutator.Float(
     Variable=InitVariable.Float().Variable
     Value=UE.InterpTo:Audio(Interp Time=@TriggerRouteTriggerRouteTime2_0.Value, Target=@n27.Output Value).Value
    ).Variable)
   n29 = VariableAccessor.Float(Variable=VariableMutator.Float(
     Variable=InitVariable.Float().Variable
     Value=UE.Multiply:Float(AdditionalOperands=@DecibelstoLinearGain_4.Linear Gain, PrimaryOperand=@n28.Value).Out
    ).Variable)
   n30 = InitVariable.Float()
   DecibelstoLinearGain_3 = UE.Decibels to Linear Gain:Float(Decibels=^Idle.Volume)
   n31 = InitVariable.Float()
   n32 = VariableAccessor.Trigger(Variable=@n26.Variable)
   n33 = VariableAccessor.Trigger(Variable=@n32.Variable)
   n34 = Value.Value:Float(Set=@n23.Value, Reset=@n33.Value) {also {defaults {Init Value=0, Target Value=1}}}
   TriggerRouteTriggerRouteTime2_1 = "TriggerRoute.Trigger Route (Time, 2):Time"(Value 1=^Combat.FadeOutTime, Value 0=^Idle.FadeInTime, Set 1=@n34.On Reset, Set 0=@n34.On Set)
   n35 = UE.InterpTo:Audio(Interp Time=@TriggerRouteTriggerRouteTime2_1.Value, Target=@n34.Output Value)
   n36 = UE.Multiply:Float(
    PrimaryOperand=@DecibelstoLinearGain_3.Linear Gain
    AdditionalOperands=VariableAccessor.Float(Variable=VariableMutator.Float(Variable=@n31.Variable, Value=@n35.Value).Variable).Value
   )
   WavePlayer_0 = UE.Wave Player:Stereo(Wave Asset=^Combat.TransitionWave, Play=@n19.Value) {also {version [1,1]}}
   WavePlayer_1 = UE.Wave Player:Stereo(Wave Asset=^Idle.TransitionWave, Play=@n21.Value) {also {version [1,1]}}
   n37 = InitVariable.Trigger()
   TriggerDelay_0 = UE.Trigger Delay(Delay Time=@TriggerRouteTriggerRouteTime2_0.Value, In=@n27.On Set)
   WavePlayer_2 = UE.Wave Player:Stereo(
    Wave Asset=^Idle.Wave
    Stop=VariableAccessor.Trigger(Variable=VariableMutator.Trigger(Variable=@n37.Variable, Value=@TriggerDelay_0.Out).Variable).Value
    Play=@n24.Value
   ) {also {defaults {Loop="<MetaSoundBuilderResult.SUCCEEDED: 0>"}, version [1,1]}}
   TriggerDelay_2 = UE.Trigger Delay(Delay Time=@TriggerRouteTriggerRouteTime2_1.Value, In=@n34.On Set)
   n38 = VariableAccessor.Trigger(Variable=VariableMutator.Trigger(Variable=InitVariable.Trigger().Variable, Value=@TriggerDelay_2.Out).Variable)
   WavePlayer_3 = UE.Wave Player:Stereo(Wave Asset=^Combat.PercWave, Stop=@n38.Value, Play=@n32.Value) {also {defaults {Loop="<MetaSoundBuilderResult.SUCCEEDED: 0>"}, version [1,1]}}
   WavePlayer_4 = UE.Wave Player:Stereo(
    Wave Asset=^Combat.ElseWave
    Stop=VariableAccessor.Trigger(Variable=@n38.Variable).Value
    Play=VariableAccessor.Trigger(Variable=@n33.Variable).Value
   ) {also {defaults {Loop="<MetaSoundBuilderResult.SUCCEEDED: 0>"}, version [1,1]}}
   AudioMixerAudioMixerStereo5_0 = "AudioMixer.Audio Mixer (Stereo, 5)"(
    Gain 3=@DecibelstoLinearGain_0.Linear Gain
    Gain 4=@DecibelstoLinearGain_1.Linear Gain
    Gain 1=@n29.Value
    Gain 2=@n29.Value
    Gain 0=VariableAccessor.Float(Variable=VariableMutator.Float(Variable=@n30.Variable, Value=@n36.Out).Variable).Value
    In 4 L=@WavePlayer_0.Out Left
    In 4 R=@WavePlayer_0.Out Right
    In 3 L=@WavePlayer_1.Out Left
    In 3 R=@WavePlayer_1.Out Right
    In 0 L=@WavePlayer_2.Out Left
    In 0 R=@WavePlayer_2.Out Right
    In 1 L=@WavePlayer_3.Out Left
    In 1 R=@WavePlayer_3.Out Right
    In 2 L=@WavePlayer_4.Out Left
    In 2 R=@WavePlayer_4.Out Right
   )
   DecibelstoLinearGain_2 = UE.Decibels to Linear Gain:Float(Decibels=^Defeated.Volume)
   AudioMixerAudioMixerStereo2_0 = "AudioMixer.Audio Mixer (Stereo, 2)"(
    In 0 L=@AudioMixerAudioMixerStereo5_0.Out L
    In 0 R=@AudioMixerAudioMixerStereo5_0.Out R
    Gain 1=@DecibelstoLinearGain_2.Linear Gain
    Gain 0=UE.InterpTo:Audio(Target=Value.Value:Float(Set=@WavePlayer_5.On Play) {also {defaults {Init Value=1, Target Value=0}}}.Output Value) {also {defaults {Interp Time=1}}}.Value
    In 1 L=@WavePlayer_5.Out Left
    In 1 R=@WavePlayer_5.Out Right
   )
   TriggerOnValueChange_0 = UE.Trigger On Value Change:Bool(Value=^MusicEncounter.bIsPaused)
   n39 = TriggerCompare.Clamp:Bool(A=^MusicEncounter.bIsPaused, Compare=@TriggerOnValueChange_0.Trigger) {also {defaults {B="<MetaSoundBuilderResult.SUCCEEDED: 0>"}}}
   n40 = Value.Value:Float(Reset=@n39.False, Set=@n39.True) {also {defaults {Init Value=1, Target Value=0}}}
   DecibelstoLinearGain_5 = UE.Decibels to Linear Gain:Float(Decibels=^MusicEncounter.OutputVolume)
   n41 = UE.Multiply:Float(
    PrimaryOperand=UE.InterpTo:Audio(Interp Time=^MusicEncounter.PauseFadeOutTime, Target=@n40.Output Value).Value
    AdditionalOperands=UE.Multiply:Float(
     AdditionalOperands=@DecibelstoLinearGain_5.Linear Gain
     PrimaryOperand=Modulation.GetModulatorValue() {also {defaults {Modulator=/Game/Sounds/Modulation/EnvelopeFollowers/FortWeapon/MP_WeaponFireEnvFollower}}}.Out
    ).Out
   )
   Multiply_0 = UE.Multiply:Audio by Float(PrimaryOperand=@AudioMixerAudioMixerStereo2_0.Out L, AdditionalOperands=@n41.Out)
   Multiply_1 = UE.Multiply:Audio by Float(PrimaryOperand=@AudioMixerAudioMixerStereo2_0.Out R, AdditionalOperands=@n41.Out)
   TriggerDelay_1 = UE.Trigger Delay(Delay Time=^MusicEncounter.PauseFadeOutTime, Reset=@n40.On Reset, In=@n40.On Set)
   n42 = InitVariable.Trigger()
   TriggerAnyTriggerAny3_0 = "TriggerAny.Trigger Any (3)"(
    In 2=@TriggerDelay_1.Out
    In 1=VariableAccessor.Trigger(Variable=VariableMutator.Trigger(Variable=@n42.Variable, Value=@TriggerSelect3_1.Out 2).Variable).Value
    In 0=@WavePlayer_5.On Finished
   )
  ]
  links [
   {dst="@UE.OutputFormat.Stereo.Audio:0", src=Multiply_0.Out}
   {dst="@UE.OutputFormat.Stereo.Audio:1", src=Multiply_1.Out}
   {dst="@UE.Source.OneShot.OnFinished", src=TriggerAnyTriggerAny3_0.Out}
  ]
 }]
properties {
 assetImportData=&self:AssetImportData
 attenuationSettings=/Game/Sounds/Attenuation/ATT_Music_NPC
 maxDistance=6500
 numChannels=2
 outputFormat=Stereo
 sampleRate=48000
 soundClassObject=&Fort_Music_NPC
}
