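# Stereo MetaSound source for a music encounter that plays a wave asset as a tempo-synced loop with distance-driven volume.
# A Harmonix metronome clock, sized by tempo and loop length in bars, drives a timestamp trigger that plays the wave.
# Master volume, fade-in, pause fade and smoothed inner-radius/falloff attenuation multiply into one gain.
# Pausing starts a delayed OnFinished matching the fade-out time, which unpausing resets.
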MetaSoundSource /MusicEncounter/MetaSounds/MSS_MusicEncounter_VolumeLoop
inputs [
 FadeInTime {type=Time}
 MasterVolume {type=Float}
 Music.LoopLength {type=Int32, access=Constructor}
 Music.Tempo {type=Float}
 Music.WaveAsset {type=WaveAsset}
 MusicEncounter.Attenuation.FalloffDistance {type=Float}
 MusicEncounter.Attenuation.InnerRadius {type=Float}
 MusicEncounter.bIsPaused {type=Bool}
 MusicEncounter.PauseFadeOutTime {type=Time}
 UE.Attenuation.Distance {type=Float}
 UE.Source.OnPlay {type=Trigger}
]
interfaces [UE.Attenuation:1.0, UE.OutputFormat.Stereo:1.0, UE.Source.OneShot:1.0, UE.Source:1.1]
outputs [
 "UE.OutputFormat.Stereo.Audio:0" {type=Audio}
 "UE.OutputFormat.Stereo.Audio:1" {type=Audio}
 UE.Source.OneShot.OnFinished {type=Trigger}
]
pages [Default {
  input_defaults {
   FadeInTime=0
   MasterVolume=1
   Music.LoopLength=4
   Music.Tempo=120
   Music.WaveAsset=null
   MusicEncounter.Attenuation.FalloffDistance=3000
   MusicEncounter.Attenuation.InnerRadius=3000
   MusicEncounter.PauseFadeOutTime=2
   MusicEncounter.bIsPaused=null
   UE.Attenuation.Distance=0
   UE.Source.OnPlay=null
  }
  nodes [
   n1 = Value.Value:Float(Set=^UE.Source.OnPlay) {also {defaults {Init Value=0, Target Value=1}}}
   n2 = UE.Add:Float(AdditionalOperands=^MusicEncounter.Attenuation.FalloffDistance, PrimaryOperand=^MusicEncounter.Attenuation.InnerRadius)
   n3 = UE.InterpTo:Audio(Target=^UE.Attenuation.Distance) {also {defaults {Interp Time=0.5}}}
   TriggerOnValueChange_0 = UE.Trigger On Value Change:Bool(Value=^MusicEncounter.bIsPaused)
   n4 = TriggerCompare.Clamp:Bool(A=^MusicEncounter.bIsPaused, Compare=@TriggerOnValueChange_0.Trigger) {also {defaults {B="<MetaSoundBuilderResult.SUCCEEDED: 0>"}}}
   n5 = Value.Value:Float(Reset=@n4.False, Set=@n4.True) {also {defaults {Init Value=1, Target Value=0}}}
   n6 = UE.Multiply:Float(
    AdditionalOperands=UE.Multiply:Float(
     PrimaryOperand=UE.InterpTo:Audio(Interp Time=^FadeInTime, Target=@n1.Output Value).Value
     AdditionalOperands=MapRange.MapRange:Float(In Range A=^MusicEncounter.Attenuation.InnerRadius, In Range B=@n2.Out, In=@n3.Value) {also {defaults {Out Range A=1, Out Range B=0}}}.Out Value
    ).Out
    PrimaryOperand=UE.Multiply:Float(
     PrimaryOperand=^MasterVolume
     AdditionalOperands=UE.InterpTo:Audio(Interp Time=^MusicEncounter.PauseFadeOutTime, Target=@n5.Output Value).Value
    ).Out
   )
   n7 = HarmonixNodes.TriggerToTransport(Play=^UE.Source.OnPlay)
   WavePlayer_0 = UE.Wave Player:Stereo(
    Wave Asset=^Music.WaveAsset
    Play=HarmonixNodes.MidiClockTimestampTrigger(MIDI Clock=HarmonixNodes.Metronome("Loop Length (Bars)"=^Music.LoopLength, Tempo=^Music.Tempo, Transport=@n7.Transport) {also {defaults {Simple Loop="<MetaSoundBuilderResult.SUCCEEDED: 0>", Time Sig. Numerator=1}}}.MIDI Clock).Trigger Out
   ) {also {version [1,1]}}
   Multiply_0 = UE.Multiply:Audio by Float(AdditionalOperands=@n6.Out, PrimaryOperand=@WavePlayer_0.Out Left)
   Multiply_1 = UE.Multiply:Audio by Float(AdditionalOperands=@n6.Out, PrimaryOperand=@WavePlayer_0.Out Right)
   TriggerDelay_0 = UE.Trigger Delay(Delay Time=^MusicEncounter.PauseFadeOutTime, Reset=@n5.On Reset, In=@n5.On Set)
  ]
  links [
   {dst="@UE.OutputFormat.Stereo.Audio:0", src=Multiply_0.Out}
   {dst="@UE.OutputFormat.Stereo.Audio:1", src=Multiply_1.Out}
   {dst="@UE.Source.OneShot.OnFinished", src=TriggerDelay_0.Out}
  ]
 }]
properties {
 assetImportData=&self:AssetImportData
 numChannels=2
 outputFormat=Stereo
 sampleRate=48000
}
